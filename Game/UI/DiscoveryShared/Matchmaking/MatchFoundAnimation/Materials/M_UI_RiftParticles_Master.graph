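# Additive UI master material that draws procedural drifting particles for the rift in the matchmaking Match Found animation.
# Panned, fitted UVs are tiled into cells, each drawing an exponential radial gradient dot that blends between two colors.
# Cosines of each cell's ceiled-coordinate length, one time-animated, provide per-cell variation signals through named reroutes, avoiding textures.
# A LinearFade mask combines vertical sine and radial gradient falloffs, and an Anim_Alpha parameter blends start and stop opacity.
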Material /Game/UI/DiscoveryShared/Matchmaking/MatchFoundAnimation/Materials/M_UI_RiftParticles_Master
graph {
 properties {bUsedWithStaticMesh=false, blendMode=BLEND_Additive, materialDomain=MD_UI, pixelDepthOffsetMode=PDOM_Legacy, usedShadingModels=Default Lit}
 nodes [
  n1 = Panner(
   Coordinate=MaterialFunctionCall() {materialFunction=&MF_UI_UVScaling}.Scaled to Fit
   Time=Multiply(A=Time(), B=ScalarParameter() {defaultValue=0.125, parameterName=TimeSpeed})
  ) {speedX=1}
  n2 = AppendVector(
   A=ScalarParameter() {defaultValue=4, group=Tiling, parameterName=TileX}
   B=ScalarParameter() {defaultValue=4, group=Tiling, parameterName=TileY}
  )
  n3 = ScalarParameter() {defaultValue=0.5, parameterName=XOffset_Start}
  n4 = ScalarParameter() {defaultValue=0.5, parameterName=XOffset_Stop}
  n5 = NamedRerouteUsage()
  n6 = ScalarParameter() {defaultValue=0.5, parameterName=YOffset_Start}
  n7 = ScalarParameter() {defaultValue=0.5, parameterName=YOffset_Stop}
  n8 = NamedRerouteUsage()
  n9 = MaterialFunctionCall(
   UVs=Frac(None=Multiply(A=@n1, B=@n2) {constB=16})
   CenterPosition=AppendVector(
    A=LinearInterpolate(A=@n3, B=@n4, Alpha=@n5) {constA=0.5, constB=0.05}
    B=LinearInterpolate(A=@n6, B=@n7, Alpha=@n8) {constA=0.5, constB=0.05}
   )
   Radius=LinearInterpolate(B=ScalarParameter() {defaultValue=0.5, parameterName=Radius}, Alpha=NamedRerouteUsage()) {constB=0.125}
   Density=LinearInterpolate(B=ScalarParameter() {defaultValue=2.333, parameterName=Density}, Alpha=NamedRerouteUsage()) {constB=0.125}
  ) {materialFunction=&RadialGradientExponential}
  LinearInterpolate(
   A=VectorParameter() {group=Color, parameterName=Color_A, defaultValue=<rgba 0 0 1 1>}.RGB
   B=VectorParameter() {group=Color, parameterName=Color_B, defaultValue=<rgba 1 0 0 1>}.RGB
   Alpha=@n9.RadialGradientExponential
  )
  n10 = Time()
  n11 = ScalarParameter() {defaultValue=0.125, parameterName=TimeSpeed}
  n12 = Frac(None=Panner(
    Coordinate=MaterialFunctionCall() {materialFunction=&MF_UI_UVScaling}.Scaled to Fit
    Time=Multiply(A=@n10, B=@n11)
   ) {speedX=1})
  n13 = AppendVector(
   A=ScalarParameter() {defaultValue=4, group=Tiling, parameterName=TileX}
   B=ScalarParameter() {defaultValue=4, group=Tiling, parameterName=TileY}
  )
  n14 = MaterialFunctionCall(Vector 2=Ceil(None=Multiply(A=@n12, B=@n13))) {materialFunction=/Engine/Functions/Engine_MaterialFunctions02/Utility/VectorLength}
  n15 = Multiply(A=@n14.V2 Length, B=Multiply(A=Time(), B=ScalarParameter() {defaultValue=0.125, group=Tiling, parameterName=Tile_AnimSpeed}))
  NamedRerouteDeclaration(None=Saturate(None=Cosine(None=@n15) {period=6.2832})) {name=OscillatingFracTiling, nodeColor=<rgba 0.6406 1 0 1>}
  NamedRerouteDeclaration(None=Saturate(None=Cosine(None=@n14.V2 Length))) {name=PanningFracTiling, nodeColor=<rgba 0.3828 1 0 1>}
  n16 = TextureCoordinate()
  n17 = Power(Base=ComponentMask(None=Sine(None=@n16) {period=2}) {g=true}, Exp=ScalarParameter() {defaultValue=1, parameterName=VerticalFadeRate})
  n18 = MaterialFunctionCall(
   UV=MaterialFunctionCall() {materialFunction=&GetUserInterfaceUV}.Normalized UV
   GradientStart=ScalarParameter() {group=LinearFade, parameterName=LinearFade_Start}
   GradientEnd=ScalarParameter() {defaultValue=1, group=LinearFade, parameterName=LinearFade_Stop}
  ) {materialFunction=&MF_UI_RadialGradient}
  n19 = ScalarParameter() {defaultValue=1, group=LinearFade, parameterName=LinearFade_EdgeHardness}
  n20 = Power(Base=Saturate(None=Multiply(A=@n18.Smooth, B=@n19)), Exp=ScalarParameter() {defaultValue=1, group=LinearFade, parameterName=LinearFade_Rate})
  NamedRerouteDeclaration(None=DotProduct(A=Saturate(None=@n17), B=Saturate(None=@n20))) {name=LinearFade, nodeColor=<rgba 0.3359 0 1 1>}
  n21 = NamedRerouteUsage()
  n22 = LinearInterpolate(
   A=ScalarParameter() {defaultValue=0.75, group=Alpha, parameterName=Alpha_Start}
   B=ScalarParameter() {defaultValue=1, group=Alpha, parameterName=Alpha_Stop}
   Alpha=ScalarParameter() {group=Alpha, parameterName=Anim_Alpha}
  )
  Saturate(None=DotProduct(A=@n9.RadialGradientExponential, B=Multiply(A=@n21, B=@n22)))
 ]
}
