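# Shared material function that draws an animated wispy line, filed under the matchmaking "match found" UI animation.
# A panning sine wave, with optional time-modulated amplitude, distorts the UVs of a second sine that forms the band, shaped by Thickness and smoothstep.
# A horizontal fade peaking mid-span masks both the TreatedWave and FullRange outputs.
# Flip Fade Direction reverses wave travel; its Length-based gradient feeds no output.
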MaterialFunction /Game/UI/DiscoveryShared/Matchmaking/MatchFoundAnimation/Materials/MF_WhispyLine_Full_Shared
graph {nodes [
  n1 = Constant() {r=1}
  n2 = Constant() {r=0.95}
  n3 = Multiply(
   A=TextureCoordinate()
   B=FunctionInput(Preview=Constant() {r=0.5}) {bUsePreviewValueAsDefault=true, inputName=Wave Frequency, inputType=FunctionInput_Scalar, sortPriority=3}
  )
  n4 = Time()
  n5 = Multiply(A=FunctionInput(Preview=Constant() {r=3}) {bUsePreviewValueAsDefault=true, inputName=Wave Speed, inputType=FunctionInput_Scalar, sortPriority=3}) {constB=-0.1}
  n6 = FunctionInput(Preview=Constant()) {bUsePreviewValueAsDefault=true, inputName=Flip Fade Direction, inputType=FunctionInput_Scalar, sortPriority=6}
  n7 = AppendVector(A=LinearInterpolate(A=@n5, B=Multiply(A=@n5) {constB=-1}, Alpha=@n6), B=Constant())
  n8 = MaterialFunctionCall(Float2=Sine(None=Panner(Coordinate=@n3, Time=@n4, Speed=@n7) {speedX=-0.125})) {materialFunction=&BreakOutFloat2Components}
  n9 = FunctionInput(Preview=Constant() {r=0.25}) {bUsePreviewValueAsDefault=true, inputName=Distortion Amount End, inputType=FunctionInput_Scalar, sortPriority=5}
  n10 = Constant() {r=2}
  n11 = Multiply(A=@n9, B=@n10)
  n12 = MaterialFunctionCall(Speed=Constant() {r=0.25}) {materialFunction=&DebugTimeSine}
  n13 = FunctionInput(Preview=Constant()) {bUsePreviewValueAsDefault=true, inputName=Modulate Distortion, inputType=FunctionInput_Scalar, sortPriority=4}
  n14 = FunctionInput(Preview=Constant()) {bUsePreviewValueAsDefault=true, inputName=Distortion Amount Start, inputType=FunctionInput_Scalar, sortPriority=4}
  n15 = Multiply(A=@n14, B=@n10)
  n16 = LinearInterpolate(
   A=LinearInterpolate(A=@n9, B=LinearInterpolate(A=@n9, B=@n11, Alpha=@n12.Result), Alpha=@n13)
   B=LinearInterpolate(A=@n14, B=LinearInterpolate(A=@n14, B=@n15, Alpha=@n12.Result), Alpha=@n13)
   Alpha=NamedRerouteUsage()
  )
  n17 = Sine(None=Add(A=TextureCoordinate(), B=Multiply(A=@n8.R, B=@n16))) {period=2}
  n18 = FunctionInput(Preview=Constant() {r=1}) {bUsePreviewValueAsDefault=true, inputName=Thickness, inputType=FunctionInput_Scalar, sortPriority=2}
  n19 = Multiply(
   A=Subtract(
    A=MaterialFunctionCall(Float2=@n17) {materialFunction=&BreakOutFloat2Components}.G
    B=OneMinus(None=@n18)
   )
   B=Divide(B=Max(A=@n18) {constB=0.001}) {constB=0}
  )
  n20 = SmoothStep(
   Min=LinearInterpolate(
    A=FunctionInput(Preview=@n1) {bUsePreviewValueAsDefault=true, inputName=Min End, inputType=FunctionInput_Scalar}
    B=FunctionInput(Preview=@n2) {bUsePreviewValueAsDefault=true, inputName=Min Start, inputType=FunctionInput_Scalar}
    Alpha=NamedRerouteUsage()
   )
   Max=FunctionInput(Preview=Constant() {r=1}) {bUsePreviewValueAsDefault=true, inputName=Max, inputType=FunctionInput_Scalar, sortPriority=1}
   Value=@n19
  )
  n21 = ComponentMask(None=TextureCoordinate()) {r=true}
  n22 = Multiply(A=Multiply(A=@n21, B=OneMinus(None=@n21))) {constB=2}
  n23 = Constant() {r=1.22}
  n24 = Saturate(None=Power(
    Base=NamedRerouteDeclaration(None=@n22) {name=LinearFadeAlpha, nodeColor=<rgba 1 0.2109 0 1>}
    Exp=FunctionInput(Preview=@n23) {bUsePreviewValueAsDefault=true, inputName=Fade Decay Rate, inputType=FunctionInput_Scalar, sortPriority=6}
   ))
  FunctionOutput(None=Saturate(None=DotProduct(A=@n20, B=@n24))) {outputName=TreatedWave}
  FunctionOutput(None=DotProduct(A=@n19, B=@n24)) {outputName=FullRange}
  n25 = FunctionInput(Preview=Constant() {r=1}) {
   bUsePreviewValueAsDefault=true
   description=Gradient end point from 0-1
   inputName=Length
   inputType=FunctionInput_Scalar
   sortPriority=2
   previewValue=<1 1 1 1>
  }
  n26 = Divide(A=@n21, B=Max(A=@n25) {constB=0.001})
  LinearInterpolate(A=OneMinus(None=@n26), B=@n26, Alpha=@n6)
 ]}
