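# Additive UI material that draws three animated wispy lines with glow, stored with the matchmaking match-found animation.
# Three calls to MF_WhispyLine_Full_Shared, each separately tuned but sharing FlipDirection, supply wave shapes and glows.
# Each wave gets a start-to-end color gradient with a randomized mix, alongside a time-driven cosine rainbow palette.
# Chained Color Dodge blends combine color, while summed alpha is shaped by Power and Final_Alpha.
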Material /Game/UI/DiscoveryShared/Matchmaking/MatchFoundAnimation/Materials/M_WhispyNoise_Full
graph {
 properties {
  bUsedWithStaticMesh=false
  blendMode=BLEND_Additive
  materialDomain=MD_UI
  numCustomizedUVs=1
  pixelDepthOffsetMode=PDOM_Legacy
  usedShadingModels=Default Lit
 }
 nodes [
  n1 = NamedRerouteUsage()
  n2 = NamedRerouteUsage()
  n3 = NamedRerouteDeclaration(None=Add(A=Add(A=@n1, B=@n2), B=NamedRerouteUsage())) {name=Alpha, nodeColor=<rgba 0.5937 1 0 1>}
  n4 = ScalarParameter() {defaultValue=1.22, parameterName=Power}
  Multiply(A=Saturate(None=Power(Base=@n3, Exp=@n4)), B=ScalarParameter() {defaultValue=1, parameterName=Final_Alpha})
  n5 = MaterialFunctionCall() {materialFunction=&GetUserInterfaceUV}
  NamedRerouteDeclaration(None=Divide(A=Multiply(A=@n5.Pixel Size, B=@n5.Pixel Size), B=@n5.Pixel Size)) {name=uv, nodeColor=<rgba 0 1 0.6016 1>}
  n6 = MaterialFunctionCall!(
   Min Start=ScalarParameter() {defaultValue=0.95, group=Wave3, parameterName=w3_MinStart}
   Min End=ScalarParameter() {defaultValue=1, group=Wave3, parameterName=w3_MinEnd}
   Max=ScalarParameter() {defaultValue=1, group=Wave3, parameterName=w3_Max}
   Length=ScalarParameter() {defaultValue=1, group=Wave3, parameterName=w3_Length}
   Thickness=ScalarParameter() {defaultValue=0.125, group=Wave3, parameterName=w3_Thickness}
   Wave Frequency=ScalarParameter() {defaultValue=0.8, group=Wave3, parameterName=w3_Frequency}
   Wave Speed=ScalarParameter() {defaultValue=2.5, group=Wave3, parameterName=w3_Speed}
   Modulate Distortion=ScalarParameter() {defaultValue=0.125, group=Wave3, parameterName=w3_ModulateDistortion}
   Distortion Amount Start=ScalarParameter() {group=Wave3, parameterName=w3_DistortionAmountStart}
   Distortion Amount End=ScalarParameter() {defaultValue=0.125, group=Wave3, parameterName=w3_DistortionAmountEnd}
   Fade Decay Rate=ScalarParameter() {defaultValue=1.22, group=Wave3, parameterName=w3_FadeDecayRate}
   Flip Fade Direction=ScalarParameter() {parameterName=FlipDirection}
  ) {materialFunction=&/MF_WhispyLine_Full_Shared}
  n7 = Saturate(None=NamedRerouteDeclaration(None=@n6.FullRange) {name=Wave3_Shape_Glow, nodeColor=<rgba 1 0.9609 0 1>})
  n8 = ScalarParameter() {defaultValue=0.125, parameterName=Wave3_GlowAmount}
  n9 = NamedRerouteDeclaration(None=@n6.TreatedWave) {name=Wave3_Shape, nodeColor=<rgba 1 0.9609 0 1>}
  n10 = ScalarParameter() {defaultValue=0.333, parameterName=Wave3_Alpha}
  NamedRerouteDeclaration(None=Add(A=Multiply(A=@n7, B=@n8), B=Multiply(A=@n9, B=@n10))) {name=Wave3, nodeColor=<rgba 1 0.1406 0 1>}
  n11 = VectorParameter() {parameterName=w1_Color_End, defaultValue=<rgba 0 0.3333 1 1>}
  n12 = NamedRerouteUsage()
  n13 = ScalarParameter() {defaultValue=0.25, parameterName=w1_RandomizedColor_MixAmount}
  NamedRerouteDeclaration(None=LinearInterpolate(
    A=LinearInterpolate(A=@n11.RGB, B=@n12, Alpha=@n13)
    B=VectorParameter() {parameterName=w1_Color_Start, defaultValue=<rgba 0.5 1 1 1>}.RGB
    Alpha=NamedRerouteUsage()
   )) {name=Wave1_Color, nodeColor=<rgba 0 1 0.6484 1>}
  n14 = VectorParameter() {parameterName=w3_Color_End, defaultValue=<rgba 0 0.3333 1 1>}
  n15 = NamedRerouteUsage()
  n16 = ScalarParameter() {defaultValue=0.25, parameterName=w3_RandomizedColor_MixAmount}
  NamedRerouteDeclaration(None=LinearInterpolate(
    A=LinearInterpolate(A=@n14.RGB, B=@n15, Alpha=@n16)
    B=VectorParameter() {parameterName=w3_Color_Start, defaultValue=<rgba 0.5 1 1 1>}.RGB
    Alpha=NamedRerouteUsage()
   )) {name=Wave3_Color, nodeColor=<rgba 0.6172 0 1 1>}
  n17 = VectorParameter() {parameterName=w2_Color_End, defaultValue=<rgba 0 0.3333 1 1>}
  n18 = NamedRerouteUsage()
  n19 = ScalarParameter() {defaultValue=0.25, parameterName=w2_RandomizedColor_MixAmount}
  NamedRerouteDeclaration(None=LinearInterpolate(
    A=LinearInterpolate(A=@n17.RGB, B=@n18, Alpha=@n19)
    B=VectorParameter() {parameterName=w2_Color_Start, defaultValue=<rgba 0.5 1 1 1>}.RGB
    Alpha=NamedRerouteUsage()
   )) {name=Wave2_Color, nodeColor=<rgba 0 0.1562 1 1>}
  n20 = Constant() {r=0.5}
  n21 = MaterialFunctionCall(Float2=NamedRerouteUsage()) {materialFunction=&BreakOutFloat2Components}
  n22 = MaterialFunctionCall(X=@n21.R, Y=@n21.R, Z=@n21.R) {materialFunction=&MakeFloat3}
  n23 = Constant3Vector() {constant=<rgba 0 2 4 0>}
  n24 = Cosine(None=Add(A=Add(A=@n22.Result, B=@n23), B=Time())) {desc=2 PI, period=6.2832}
  NamedRerouteDeclaration(None=Add(A=@n20, B=Multiply(A=@n20, B=@n24))) {name=ModulatingColorRainbow, nodeColor=<rgba 1 0.6328 0 1>}
  n25 = MaterialFunctionCall(
   Min Start=ScalarParameter() {defaultValue=0.95, group=Wave1, parameterName=w1_MinStart}
   Min End=ScalarParameter() {defaultValue=1, group=Wave1, parameterName=w1_MinEnd}
   Max=ScalarParameter() {defaultValue=1, group=Wave1, parameterName=w1_Max}
   Length=ScalarParameter() {defaultValue=1, group=Wave1, parameterName=w1_Length}
   Thickness=ScalarParameter() {defaultValue=0.125, group=Wave1, parameterName=w1_Thickness}
   Wave Frequency=ScalarParameter() {defaultValue=1, group=Wave1, parameterName=w1_Frequency}
   Wave Speed=ScalarParameter() {defaultValue=3, group=Wave1, parameterName=w1_Speed}
   Modulate Distortion=ScalarParameter() {defaultValue=0.5, group=Wave1, parameterName=w1_ModulateDistortion}
   Distortion Amount Start=ScalarParameter() {defaultValue=0.05, group=Wave1, parameterName=w1_DistortionAmountStart}
   Distortion Amount End=ScalarParameter() {defaultValue=0.125, group=Wave1, parameterName=w1_DistortionAmountEnd}
   Fade Decay Rate=ScalarParameter() {defaultValue=1.22, group=Wave1, parameterName=w1_FadeDecayRate}
   Flip Fade Direction=ScalarParameter() {parameterName=FlipDirection}
  ) {materialFunction=&MF_WhispyLine_Full_Shared}
  n26 = Saturate(None=NamedRerouteDeclaration(None=@n25.FullRange) {name=Wave1_Shape_Glow, nodeColor=<rgba 1 0.9141 0 1>})
  n27 = ScalarParameter() {defaultValue=0.125, parameterName=Wave1_GlowAmount}
  n28 = NamedRerouteDeclaration(None=@n25.TreatedWave) {name=Wave1_Shape, nodeColor=<rgba 1 0.9141 0 1>}
  n29 = ScalarParameter() {defaultValue=0.333, parameterName=Wave1_Alpha}
  NamedRerouteDeclaration(None=Add(A=Multiply(A=@n26, B=@n27), B=Multiply(A=@n28, B=@n29))) {name=Wave1, nodeColor=<rgba 0.0312 0 1 1>}
  n30 = MaterialFunctionCall(
   Min Start=ScalarParameter() {defaultValue=0.95, group=Wave2, parameterName=w2_MinStart}
   Min End=ScalarParameter() {defaultValue=1, group=Wave2, parameterName=w2_MinEnd}
   Max=ScalarParameter() {defaultValue=1, group=Wave2, parameterName=w2_Max}
   Length=ScalarParameter() {defaultValue=1, group=Wave2, parameterName=w2_Length}
   Thickness=ScalarParameter() {defaultValue=0.125, group=Wave2, parameterName=w2_Thickness}
   Wave Frequency=ScalarParameter() {defaultValue=0.9, group=Wave2, parameterName=w2_Frequency}
   Wave Speed=ScalarParameter() {defaultValue=2.75, group=Wave2, parameterName=w2_Speed}
   Modulate Distortion=ScalarParameter() {defaultValue=0.25, group=Wave2, parameterName=w2_ModulateDistortion}
   Distortion Amount Start=ScalarParameter() {group=Wave2, parameterName=w2_DistortionAmountStart}
   Distortion Amount End=ScalarParameter() {defaultValue=0.2, group=Wave2, parameterName=w2_DistortionAmountEnd}
   Fade Decay Rate=ScalarParameter() {defaultValue=1.5, group=Wave2, parameterName=w2_FadeDecayRate}
   Flip Fade Direction=ScalarParameter() {parameterName=FlipDirection}
  ) {materialFunction=&MF_WhispyLine_Full_Shared}
  n31 = Saturate(None=NamedRerouteDeclaration(None=@n30.FullRange) {name=Wave2_Shape_Glow, nodeColor=<rgba 1 0 0.375 1>})
  n32 = ScalarParameter() {defaultValue=0.125, parameterName=Wave2_GlowAmount}
  n33 = NamedRerouteDeclaration(None=@n30.TreatedWave) {name=Wave2_Shape, nodeColor=<rgba 1 0 0.375 1>}
  n34 = ScalarParameter() {defaultValue=0.333, parameterName=Wave2_Alpha}
  NamedRerouteDeclaration(None=Add(A=Multiply(A=@n31, B=@n32), B=Multiply(A=@n33, B=@n34))) {name=Wave2, nodeColor=<rgba 1 0.5156 0 1>}
  n35 = MaterialFunctionCall!(Blend=NamedRerouteUsage(), Base=NamedRerouteUsage()) {materialFunction=/Engine/Functions/Engine_MaterialFunctions03/Blends/Blend_ColorDodge}
  Saturate(None=MaterialFunctionCall(Blend=Saturate(None=@n35.Result), Base=NamedRerouteUsage()) {materialFunction=&Blend_ColorDodge}.Result)
 ]
}
